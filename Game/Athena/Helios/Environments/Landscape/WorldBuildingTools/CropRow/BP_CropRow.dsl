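; Crop row world-building actor that pairs a spline with a PCG component, with empty gameplay events.
; The construction script applies the PCG regenerate-in-editor and closed-loop spline settings, then places a billboard at a Z offset above the first spline point.
; Simple destructible prop functions expose the instanced static mesh component, return a wood-type entry, and report that the source actor should not be kept.
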(event EventBeginPlay)

(event Collision|EventActorBeginOverlap
  (params (OtherActor :type Actor%20Object%20Reference)))

(event EventDestroyed)

(fn GetSimpleDestructiblePropComponents
  (bind _returnvalue (Actor|GetComponentbyClass "/Script/Engine.InstancedStaticMeshComponent"))
  (if (Utilities|IsValid _returnvalue :!class /Script/Engine.KismetSystemLibrary)
    (return (Utilities|Array|MakeArray _returnvalue))))

(fn GetSimpleDestructiblePropEntry
  (params (SourceIndex :type Integer) (SourceComponent :type Static%20Mesh%20Component%20Object%20Reference))
  (return 0 3 "(Name=\"\")" 0 "(CurveTable=None,RowName=\"\")" "Wood" "Wood" "ATTACH_None" "SLOT_None" false false false))

(fn SetBillBoardLocation
  (Transformation|SetWorldLocation (Variables|Default|GetBillboard) (+ (Spline|GetLocationatSplinePoint (Variables|Default|GetSpline) 0 "World") (Math|Vector|MakeVector 0.0 0.0 (Variables|Default|GetBillboardZOffset))) false false))

(fn ShouldKeepSourceActor
  (return false))

(fn ConstructionScript
  (Variables|/Script/PCG.PCGComponent|SetbRegenerateInEditor (Variables|Default_Parameters|GetRegenerate%20PCG%20Volume) (Variables|Default|GetPCG))
  (Spline|SetClosedLoop (Variables|Default|GetSpline) (Variables|Default_Parameters|GetUseClosedLoopSpline) true)
  (CallFunction|SetBillBoardLocation))
