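; PCG graph that plants crop rows, citrus trees by default, at a set spacing along splines tagged CropSpline.
; Sampled points project onto the landscape, then get randomized offset, yaw and scale.
; A noise-driven Z-bias lerps each rotation from the projected orientation toward an absolute, upright one.
; The SpawnActors flag switches output between attached actors and hierarchical instanced static meshes with custom collision.
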(graph PCG_CropRow
  (params (Distance :type Int :default 150) (RotationMin :type Double :default -45.0) (RotationMax :type Double :default 45.0) (ScaleMin :type Double :default 1.0) (ScaleMax :type Double :default 1.1) (Offsetmin :type Double :default -32.0) (OffsetMax :type Double :default 32.0) (Z-BiasMin :type Double :default 0.5) (Z-BiasMax :type Double :default 1.0) (Mesh :type (Object /Script/Engine.StaticMesh) :default /Game/Environments/Asteria/Foliage/ChaparralPlants/CitrusTree/Meshes/SM_Helios_ChapPlant_CitrusTree.SM_Helios_ChapPlant_CitrusTree) (ActorToSpawn :type (SoftClass /Script/Engine.Actor) :default /Game/Environments/Asteria/Foliage/ChaparralPlants/CitrusTree/Blueprints/BP_Helios_ChapPlant_CitrusTree.BP_Helios_ChapPlant_CitrusTree_C) (SpawnActors :type Bool))
  (bind GetSplineData_0 (Get%20Spline%20Data :bComponentsMustOverlapSelf false))
  (bind FilterByTag_1 (Filter%20Data%20By%20Tag :bTokenizeOnWhiteSpace true :SelectedTags "CropSpline"))
  (bind Debug_5 (Debug))
  (bind SplineSampler_6 (Spline%20Sampler :SamplerParams (struct :Mode Distance :SubdivisionsPerSegment 10 :DistanceIncrement 512.0) :DistanceIncrement Distance))
  (bind GetLandscapeData_25 (Get%20Landscape%20Data :bComponentsMustOverlapSelf false :SamplingProperties (struct)))
  (bind Projection_26 (Projection))
  (bind ToPoint_32 (To%20Point))
  (bind AttributeNoise_8 (Attribute%20Noise))
  (bind CreateAttribute_10 (Create%20Constant :AttributeTypes (struct :Type Float :VectorValue (1.0 0.0 0.0)) :OutputTarget Zero))
  (bind MakeRotatorAttribute_40 (Make%20Rotator%20Attribute :Operation MakeRotFromAngles :Yaw RotationMax))
  (bind MakeRotatorAttribute_42 (Make%20Rotator%20Attribute :Operation MakeRotFromAngles :Yaw RotationMin))
  (bind MakeVectorAttribute_43 (Make%20Vector%20Attribute :OutputType Vector :X Offsetmin :Y Offsetmin))
  (bind MakeVectorAttribute_46 (Make%20Vector%20Attribute :OutputType Vector :X OffsetMax :Y OffsetMax))
  (bind TransformPoints_13 (Transform%20Points :OffsetMax (4.0 4.0 4.0) :OffsetMin (-4.0 -4.0 -4.0) :RotationMax (struct :Yaw 2e+01) :RotationMin (struct :Yaw -2e+01)))
  (bind AttributeNoise_10 (Attribute%20Noise))
  (bind ToPoint_27 (To%20Point))
  (bind Debug_12 (Debug))
  (bind TransformPoints_24 (Transform%20Points :ScaleMin (0.8 0.8 0.8) :ScaleMax-param ScaleMax :ScaleMin-param ScaleMin))
  (bind AttributeNoise_12 (Attribute%20Noise))
  (bind AttributeMathsOp_6 (Attribute%20Maths%20Op :InputSource3 $Density :Operation Lerp :OutputTarget ration :InA Z-BiasMin :InB Z-BiasMax))
  (bind TransformPoints_1 (Transform%20Points :bAbsoluteRotation true :ScaleMin (0.8 0.8 0.8)))
  (bind AttributeRotatorOp_0 (Attribute%20Rotator%20Op :InputSource1 $Rotation :InputSource2 $Rotation :Operation Lerp))
  (bind Branch_7 (Branch :bOutputToB-param SpawnActors))
  (bind CreateAttribute_5 (Create%20Constant :AttributeTypes (struct :Type Vector :VectorValue (0.0 0.0 1.0)) :OutputTarget WorldUp))
  (bind CreateAttribute_9 (Create%20Constant :AttributeTypes (struct :Type Vector :VectorValue (0.0 1.0 0.0)) :OutputTarget WorldRight))
  (bind CreateAttribute_11 (Create%20Constant :AttributeTypes (struct :Type Vector :VectorValue (1.0 0.0 0.0)) :OutputTarget WorldForward))
  (bind MakeRotatorAttribute_7 (Make%20Rotator%20Attribute))
  ; Spawn Actor
  (bind CopyAttribute_3 (Copy%20Attributes :InputSource ActorToSpawn :OutputTarget ActorToSpawn :Source ActorToSpawn))
  (bind SpawnActor_1 (Spawn%20Actor :AttachOptions Attached :bSpawnByAttribute true :SpawnAttribute ActorToSpawn))
  ; Spawn ISM
  (bind CopyAttribute_10 (Copy%20Attributes :InputSource Mesh :OutputTarget Mesh :Source Mesh))
  (bind StaticMeshSpawner_11 (Static%20Mesh%20Spawner :MeshSelectorParameters (class:/Script/PCG.PCGMeshSelectorByAttribute :AttributeName Mesh :TemplateDescriptor (struct :ComponentClass /Script/Engine.HierarchicalInstancedStaticMeshComponent :Mobility Movable :BodyInstance (struct :CollisionProfileName FortBuildingMeshPhysics :CollisionResponses (struct :ResponseArray [(struct :Channel FCC_Trace_Interaction) (struct :Channel FCC_Trace_WeakSpot :Response ECR_Overlap) (struct :Channel FCC_Trace_WeaponCapsule) (struct :Channel FCC_Object_PawnAI) (struct :Channel FCC_Trace_WeaponMulti) (struct :Channel FCC_Trace_Weapon) (struct :Channel FCC_Object_Projectile) (struct :Channel FCC_Object_ProjectileIgnoreAI) (struct :Channel FCC_Trace_World) (struct :Channel FCC_Object_ProjectileMesh) (struct :Channel FCC_Trace_WeaponPawnAndWorldMulti :Response ECR_Overlap) (struct :Channel FCC_Trace_Audio)]) :bAutoWeld true) :bGenerateOverlapEvents true :WorldPositionOffsetDisableDistance 1024)) :MeshSelectorType /Script/PCG.PCGMeshSelectorByAttribute))
  (connect AttributeNoise_12 Out AttributeMathsOp_6 Ratio)
  (connect ToPoint_32 Out AttributeNoise_8 In)
  (connect TransformPoints_13 Out AttributeNoise_10 In)
  (connect TransformPoints_24 Out AttributeNoise_12 In)
  (connect TransformPoints_24 Out AttributeRotatorOp_0 InA)
  (connect TransformPoints_1 Out AttributeRotatorOp_0 InB)
  (connect AttributeMathsOp_6 Out AttributeRotatorOp_0 Ratio)
  (connect AttributeRotatorOp_0 Out Branch_7 In)
  (connect Branch_7 Output%20B CopyAttribute_3 Target)
  (connect Branch_7 Output%20A CopyAttribute_10 Target)
  (connect ToPoint_27 Out Debug_12 In)
  (connect GetSplineData_0 Out FilterByTag_1 In)
  (connect CreateAttribute_11 Out MakeRotatorAttribute_7 Forward)
  (connect CreateAttribute_9 Out MakeRotatorAttribute_7 Right)
  (connect CreateAttribute_5 Out MakeRotatorAttribute_7 Up)
  (connect CreateAttribute_10 Out MakeRotatorAttribute_40 Pitch)
  (connect CreateAttribute_10 Out MakeRotatorAttribute_40 Roll)
  (connect CreateAttribute_10 Out MakeRotatorAttribute_42 Pitch)
  (connect CreateAttribute_10 Out MakeRotatorAttribute_42 Roll)
  (connect SplineSampler_6 Out Projection_26 In)
  (connect GetLandscapeData_25 Out Projection_26 Projection%20Target)
  (connect CopyAttribute_3 Out SpawnActor_1 In)
  (connect FilterByTag_1 InsideFilter SplineSampler_6 Spline)
  (connect CopyAttribute_10 Out StaticMeshSpawner_11 In)
  (connect Projection_26 Out ToPoint_27 In)
  (connect Projection_26 Out ToPoint_32 In)
  (connect TransformPoints_24 Out TransformPoints_1 In)
  (connect MakeRotatorAttribute_40 Out TransformPoints_1 RotationMax)
  (connect MakeRotatorAttribute_42 Out TransformPoints_1 RotationMin)
  (connect AttributeNoise_8 Out TransformPoints_13 In)
  (connect MakeVectorAttribute_46 Out TransformPoints_13 OffsetMax)
  (connect MakeVectorAttribute_43 Out TransformPoints_13 OffsetMin)
  (connect MakeRotatorAttribute_40 Out TransformPoints_13 RotationMax)
  (connect MakeRotatorAttribute_42 Out TransformPoints_13 RotationMin)
  (connect AttributeNoise_10 Out TransformPoints_24 In))
